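# Physics asset for the Cloaked Male backpack cosmetic mesh, defining simulation bodies for two sword bones and a hanging tassel.
# A small box body on the pack bone anchors both sword capsules through tight angular limits, so they can only sway slightly.
# A three-capsule tassel chain hangs from the second sword, with looser swing limits that widen after the first link and tight twist throughout.
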PhysicsAsset /BRCosmetics/Accessories/FORT_Backpacks/Backpack_M_MED_Cloaked_Male/Meshes/M_MED_Cloaked_Male_Physics
bodies {
 dyn_sword_a [{shapeType=Capsule, length=20.0786, radius=2.4163, center=<10.4811 0 0>, rotation=<rot 90 0 0>}]
 dyn_sword_b [{shapeType=Capsule, length=11.2708, radius=2.9052, center=<7.6053 0 0>, rotation=<rot 90 0 0>}]
 dyn_tassel_01 [{shapeType=Capsule, length=4.5197, radius=0.9564, center=<3.004 0 0>, rotation=<rot 90 0 0>}]
 dyn_tassel_02 [{shapeType=Capsule, length=4.5197, radius=0.9564, center=<2.9969 0 0>, rotation=<rot 90 0 0>}]
 dyn_tassel_03 [{shapeType=Capsule, length=3.1339, radius=2.1314, center=<3.5837 0 0>, rotation=<rot 90 0 0>}]
 pack [{shapeType=Box, extentX=1, extentY=1.4194, extentZ=1.3885}]
}
constraints [
 {bone1Name=dyn_tassel_02, bone2Name=dyn_tassel_01, swing1LimitDegrees=35, swing2LimitDegrees=35, twistLimitDegrees=5}
 {bone1Name=dyn_tassel_03, bone2Name=dyn_tassel_02, swing1LimitDegrees=35, swing2LimitDegrees=35, twistLimitDegrees=5}
 {bone1Name=dyn_sword_b, bone2Name=pack, swing1LimitDegrees=5, swing2LimitDegrees=5, twistLimitDegrees=2}
 {bone1Name=dyn_sword_a, bone2Name=pack, swing1LimitDegrees=5, swing2LimitDegrees=5, twistLimitDegrees=2}
 {bone1Name=dyn_tassel_01, bone2Name=dyn_sword_b, swing1LimitDegrees=20, swing2LimitDegrees=20, twistLimitDegrees=5}
]
